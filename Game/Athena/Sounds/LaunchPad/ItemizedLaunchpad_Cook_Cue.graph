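# ItemizedLaunchpad_Cook_Cue is a Sound Cue that plays the itemized launch pad's cook sound in Athena.
# A single wave player for the ItemizedLaunchpad_Cook wave feeds a modulator at the root.
# It uses the Fort_Traps sound class and borrows the default grenade cook attenuation.
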SoundCue /Game/Athena/Sounds/LaunchPad/ItemizedLaunchpad_Cook_Cue
root=Modulator_0
nodes [
 {id=Modulator_0, children [WavePlayer_1]}
 {id=WavePlayer_1, properties {soundWaveAssetPtr=&/ItemizedLaunchpad_Cook}}
]
properties {
 attenuationSettings=/Game/Sounds/Fort_Weapons/Grenades/Grenade_Default_Cook_Att
 maxDistance=3000
 soundClassObject=&Fort_Traps
}
