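# Sound cue for throwing the itemized launchpad in Athena.
# A single wave player feeds the ItemizedLaunchpad_Throw sound through a modulator at the root.
# It uses the traps sound class and traps attenuation settings for spatial falloff.
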SoundCue /Game/Athena/Sounds/LaunchPad/ItemizedLaunchpad_Throw_Cue
root=Modulator_0
nodes [
 {id=Modulator_0, children [WavePlayer_0]}
 {id=WavePlayer_0, properties {soundWaveAssetPtr=&/ItemizedLaunchpad_Throw}}
]
properties {
 attenuationSettings=&Fort_Traps_Att
 maxDistance=6000
 soundClassObject=&Fort_Traps
}
