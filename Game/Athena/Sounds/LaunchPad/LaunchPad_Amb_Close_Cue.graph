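# Sound cue for the close-range ambient sound of the Athena launch pad, assigned to the looping loot ambient sound class.
# A single wave player feeds an Enveloper, whose curves ramp pitch up and fade volume to zero over a few seconds.
# The cue plays at reduced volume under the loot ambient concurrency rule.
# It stores attenuation override values for logarithmic falloff, listener focus and stereo spread.
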SoundCue /Game/Athena/Sounds/LaunchPad/LaunchPad_Amb_Close_Cue
root=Enveloper_1
nodes [
 {
  id=Enveloper_1
  children [WavePlayer_1]
  properties {
   pitchCurve {editorCurveData {keys [{interpMode=RCIM_Cubic, time=0}, <at 2 2>]}}
   volumeCurve {editorCurveData {keys [<at 0.0001 0.7563 RCIM_Cubic>,<at 3 0>]}}
  }
 }
 {id=WavePlayer_1, properties {soundWaveAssetPtr=&/LaunchPad_Amb_Close}}
]
properties {
 maxDistance=2097152
 soundClassObject=/Game/Sounds/FortSoundClassesAndModes/Fort_Loot_Ambient_Loop
 volumeMultiplier=0.3
 attenuationOverrides {
  bEnableListenerFocus=true
  distanceAlgorithm=Logarithmic
  falloffDistance=681.9642
  nonFocusVolumeAttenuation=0.4762
  stereoSpread=35.6359
  attenuationShapeExtents {x=1000}
 }
 concurrencySet [/Game/Sounds/Concurrency_Rules/Loot_Ambient_Conc]
}
