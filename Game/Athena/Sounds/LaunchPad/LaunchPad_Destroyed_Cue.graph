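# Sound cue for the launch pad trap being destroyed.
# A random node picks one of two destroy waves with equal weight and feeds it through a modulator at default settings.
# It plays in the traps sound class with trap attenuation and reduced volume, and its concurrency override blocks new instances once the limit is reached rather than stopping playing ones.
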SoundCue /Game/Athena/Sounds/LaunchPad/LaunchPad_Destroyed_Cue
root=Modulator_12
nodes [
 {id=Modulator_12, children [Random_31]}
 {id=Random_31, children [WavePlayer_110, WavePlayer_112], properties {weights [1,1]}}
 {id=WavePlayer_110, properties {soundWaveAssetPtr=&/LaunchPad_Destroy_001}}
 {id=WavePlayer_112, properties {soundWaveAssetPtr=&/LaunchPad_Destroy_002}}
]
properties {
 attenuationSettings=&Fort_Traps_Att
 bOverrideConcurrency=true
 maxDistance=6000
 soundClassObject=&Fort_Traps
 volumeMultiplier=0.5
 concurrencyOverrides {resolutionRule=PreventNew}
}
