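# Sound cue for the third-person launch pad jump of a vehicle, played as a trap-class sound with trap attenuation.
# A quality-level root selects one of four branches.
# The first two each pass an equally weighted random pick of two jump waves through a pair of chained modulators.
# The last two each play a single optimized third-person wave through one modulator, dropping the random variation.
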SoundCue /Game/Athena/Sounds/LaunchPad/LaunchPad_Launch_Vehicle_3P_Cue
root=QualityLevel_0
nodes [
 {id=Modulator_0, children [Modulator_1]}
 {id=Modulator_1, children [Random_2]}
 {id=Modulator_10, children [WavePlayer_22]}
 {id=Modulator_2, children [Modulator_3]}
 {id=Modulator_3, children [Random_0]}
 {id=Modulator_9, children [WavePlayer_21]}
 {id=QualityLevel_0, children [Modulator_0,Modulator_2,Modulator_10,Modulator_9]}
 {id=Random_0, children [WavePlayer_0,WavePlayer_1], properties {weights [1,1]}}
 {id=Random_2, children [WavePlayer_6,WavePlayer_7], properties {weights [1,1]}}
 {id=WavePlayer_0, properties {soundWaveAssetPtr=&/SFX_LaunchPad_Jump_01}}
 {id=WavePlayer_1, properties {soundWaveAssetPtr=&/SFX_LaunchPad_Jump_02}}
 {id=WavePlayer_21, properties {soundWaveAssetPtr=&/SFX_LaunchPad_Jump_3P_opt}}
 {id=WavePlayer_22, properties {soundWaveAssetPtr=&SFX_LaunchPad_Jump_3P_opt}}
 {id=WavePlayer_6, properties {soundWaveAssetPtr=&SFX_LaunchPad_Jump_01}}
 {id=WavePlayer_7, properties {soundWaveAssetPtr=&SFX_LaunchPad_Jump_02}}
]
properties {
 attenuationSettings=&Fort_Traps_Att
 maxDistance=6000
 soundClassObject=&Fort_Traps
 volumeMultiplier=0.5
}
