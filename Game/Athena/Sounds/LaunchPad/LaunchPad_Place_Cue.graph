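# Sound cue for placing a launch pad, using the traps sound class and attenuation.
# A mixer layers the main placement sound at full volume with a much quieter deploy layer.
# The main layer switches by audio quality level between the standard wave for the first two levels and an "opt" variant for the last two.
# The deploy layer randomly picks one of three equally weighted waves and raises it to a fixed higher pitch.
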SoundCue /Game/Athena/Sounds/LaunchPad/LaunchPad_Place_Cue
root=Mixer_0
nodes [
 {id=Mixer_0, children [QualityLevel_0, Modulator_0], properties {inputVolume [1,0.2]}}
 {id=Modulator_0, children [Random_0], properties {pitchMax=1.3, pitchMin=1.3}}
 {id=QualityLevel_0, children [WavePlayer_3,WavePlayer_3,WavePlayer_0,WavePlayer_0]}
 {id=Random_0, children [WavePlayer_1,WavePlayer_2,WavePlayer_4], properties {weights [1,1,1]}}
 {id=WavePlayer_0, properties {soundWaveAssetPtr=&/sfx_LaunchPad_Place_opt}}
 {id=WavePlayer_1, properties {soundWaveAssetPtr=&/ItemizedLaunchpad_Deploy_01}}
 {id=WavePlayer_2, properties {soundWaveAssetPtr=&/ItemizedLaunchpad_Deploy_02}}
 {id=WavePlayer_3, properties {soundWaveAssetPtr=&/sfx_LaunchPad_Place}}
 {id=WavePlayer_4, properties {soundWaveAssetPtr=&/ItemizedLaunchpad_Deploy_03}}
]
properties {
 attenuationSettings=&Fort_Traps_Att
 maxDistance=6000
 soundClassObject=&Fort_Traps
}
