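# Sound cue for picking up a Launch Pad in Athena.
# A single wave player feeds the Launch Pad UI select wave straight to the output at reduced volume.
# It plays in the gadgets and items sound class and borrows the weapon drop attenuation settings.
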SoundCue /Game/Athena/Sounds/LaunchPad/Launch_Pad_Pickup_Cue
root=WavePlayer_0
nodes [{id=WavePlayer_0, properties {soundWaveAssetPtr=&/Launch_Pad_UI_Select}}]
properties {
 attenuationSettings=&Fort_Weapon_Drop_Att
 maxDistance=3000
 soundClassObject=&Fort_Gadgets_Items
 volumeMultiplier=0.2
}
